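# Sound cue for one-shot nighttime SteelPecker bird calls, an outdoor ambient spot sound in the Asteria rainforest biome.
# A Random node picks among five equally weighted call recordings, then a Modulator lowers pitch and slightly varies volume.
# All quality levels share that same branch.
# The cue uses bird one-shot attenuation and the outdoor ambient one-shot class, with a send to the exterior ambient source bus.
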SoundCue /Game/Sounds/Asteria/Biomes/Rainforest/Spots/Spot_Bird_Night_SteelPecker_Calls_Cue
root=QualityLevel_0
nodes [
 {id=Modulator_0, children [Random_1], properties {pitchMax=0.9, pitchMin=0.6, volumeMax=1, volumeMin=0.8}}
 {id=QualityLevel_0, children [Modulator_0,Modulator_0,Modulator_0,Modulator_0]}
 {id=Random_1, children [WavePlayer_0,WavePlayer_1,WavePlayer_2,WavePlayer_3,WavePlayer_4], properties {weights [1,1,1,1,1]}}
 {id=WavePlayer_0, properties {soundWaveAssetPtr=&/Spot_Bird_Night_SteelPecker_Calls_01}}
 {id=WavePlayer_1, properties {soundWaveAssetPtr=&/Spot_Bird_Night_SteelPecker_Calls_02}}
 {id=WavePlayer_2, properties {soundWaveAssetPtr=&/Spot_Bird_Night_SteelPecker_Calls_03}}
 {id=WavePlayer_3, properties {soundWaveAssetPtr=&/Spot_Bird_Night_SteelPecker_Calls_04}}
 {id=WavePlayer_4, properties {soundWaveAssetPtr=&/Spot_Bird_Night_SteelPecker_Calls_05}}
]
properties {
 attenuationSettings=&Fort_TileAmbient_OneShot_Bird_Att
 maxDistance=8005
 soundClassObject=&Fort_Ambient_OneShots_Outdoors
 volumeMultiplier=0.251
 busSends [{
   audioBus=null
   maxSendDistance=5000
   maxSendLevel=1.2
   minSendDistance=3000
   minSendLevel=0.2
   sendLevel=0
   soundSourceBus=&AmbientAudio_Exterior_SourceBus
  }]
 preEffectBusSends [{}]
}
