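# Sound cue for a tropical night SteelPecker bird croak, an outdoor ambient one-shot spot in the Asteria rainforest biome.
# An equal-weight random node picks one of five croak waves, then a modulator varies pitch and volume.
# All four quality-level inputs share that chain.
# Playback uses bird one-shot attenuation at reduced volume, with a distance-scaled send to the exterior ambient source bus.
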SoundCue /Game/Sounds/Asteria/Biomes/Rainforest/Spots/Spot_Bird_Tropical_Night_SteelPecker_Croak_Cue
root=QualityLevel_0
nodes [
 {id=Modulator_1, children [Random_1], properties {pitchMax=1.1, pitchMin=0.8, volumeMax=1.2, volumeMin=1}}
 {id=QualityLevel_0, children [Modulator_1,Modulator_1,Modulator_1,Modulator_1]}
 {id=Random_1, children [WavePlayer_0,WavePlayer_1,WavePlayer_2,WavePlayer_3,WavePlayer_4], properties {weights [1,1,1,1,1]}}
 {id=WavePlayer_0, properties {soundWaveAssetPtr=&/Spot_Bird_Tropical_Night_SteelPecker_Croak_03}}
 {id=WavePlayer_1, properties {soundWaveAssetPtr=&/Spot_Bird_Tropical_Night_SteelPecker_Croak_05}}
 {id=WavePlayer_2, properties {soundWaveAssetPtr=&/Spot_Bird_Tropical_Night_SteelPecker_Croak_01}}
 {id=WavePlayer_3, properties {soundWaveAssetPtr=&/Spot_Bird_Tropical_Night_SteelPecker_Croak_02}}
 {id=WavePlayer_4, properties {soundWaveAssetPtr=&/Spot_Bird_Tropical_Night_SteelPecker_Croak_04}}
]
properties {
 attenuationSettings=&Fort_TileAmbient_OneShot_Bird_Att
 maxDistance=8005
 soundClassObject=&Fort_Ambient_OneShots_Outdoors
 volumeMultiplier=0.251
 busSends [{
   audioBus=null
   maxSendDistance=5000
   maxSendLevel=1.2
   minSendDistance=3000
   minSendLevel=0.2
   sendLevel=0
   soundSourceBus=&AmbientAudio_Exterior_SourceBus
  }]
 preEffectBusSends [{}]
}
