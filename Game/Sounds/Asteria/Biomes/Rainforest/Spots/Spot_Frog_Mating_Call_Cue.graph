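# Sound cue for a frog mating call spot one-shot in the Asteria rainforest biome ambience.
# An equally weighted random node picks among eleven plain and pitched call recordings, then a modulator varies pitch downward and volume upward slightly.
# All four quality-level inputs share that same chain.
# It uses a bird one-shot attenuation and outdoor ambient sound class, with a distance-scaled send to the exterior ambient source bus.
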SoundCue /Game/Sounds/Asteria/Biomes/Rainforest/Spots/Spot_Frog_Mating_Call_Cue
root=QualityLevel_0
nodes [
 {id=Modulator_0, children [Random_2], properties {pitchMax=1, pitchMin=0.8, volumeMax=1.2, volumeMin=1}}
 {id=QualityLevel_0, children [Modulator_0,Modulator_0,Modulator_0,Modulator_0]}
 {
  id=Random_2
  children [
   WavePlayer_16
   WavePlayer_17
   WavePlayer_18
   WavePlayer_19
   WavePlayer_20
   WavePlayer_21
   WavePlayer_22
   WavePlayer_23
   WavePlayer_25
   WavePlayer_26
   WavePlayer_27
  ]
  properties {weights [1,1,1,1,1,1,1,1,1,1,1]}
 }
 {id=WavePlayer_16, properties {soundWaveAssetPtr=&/Spot_Frog_Mating_Call_01}}
 {id=WavePlayer_17, properties {soundWaveAssetPtr=&/Spot_Frog_Mating_Call_02}}
 {id=WavePlayer_18, properties {soundWaveAssetPtr=&/Spot_Frog_Mating_Call_03}}
 {id=WavePlayer_19, properties {soundWaveAssetPtr=&/Spot_Frog_Mating_Call_05}}
 {id=WavePlayer_20, properties {soundWaveAssetPtr=&/Spot_Frog_Mating_Call_06}}
 {id=WavePlayer_21, properties {soundWaveAssetPtr=&/Spot_Frog_Mating_Call_Pitched_01}}
 {id=WavePlayer_22, properties {soundWaveAssetPtr=&/Spot_Frog_Mating_Call_Pitched_02}}
 {id=WavePlayer_23, properties {soundWaveAssetPtr=&/Spot_Frog_Mating_Call_Pitched_03}}
 {id=WavePlayer_25, properties {soundWaveAssetPtr=&/Spot_Frog_Mating_Call_Pitched_05}}
 {id=WavePlayer_26, properties {soundWaveAssetPtr=&/Spot_Frog_Mating_Call_Pitched_06}}
 {id=WavePlayer_27, properties {soundWaveAssetPtr=&/Spot_Frog_Mating_Call_04}}
]
properties {
 attenuationSettings=&Fort_TileAmbient_OneShot_Bird_Att
 maxDistance=8005
 soundClassObject=&Fort_Ambient_OneShots_Outdoors
 volumeMultiplier=0.401
 busSends [{
   audioBus=null
   maxSendDistance=5000
   maxSendLevel=1.2
   minSendDistance=3000
   minSendLevel=0.2
   sendLevel=0
   soundSourceBus=&AmbientAudio_Exterior_SourceBus
  }]
 preEffectBusSends [{}]
}
