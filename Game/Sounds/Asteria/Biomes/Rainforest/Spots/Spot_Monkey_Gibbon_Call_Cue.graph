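# Sound cue for one-shot gibbon call spot ambience in the Asteria rainforest biome.
# An equal-weight random node picks one of six call recordings, and a modulator varies pitch and boosts volume per play.
# All four quality levels share that same chain.
# It uses bird one-shot tile attenuation and the outdoor ambient one-shot class, with a distance-ranged send to the exterior ambient source bus.
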SoundCue /Game/Sounds/Asteria/Biomes/Rainforest/Spots/Spot_Monkey_Gibbon_Call_Cue
root=QualityLevel_0
nodes [
 {id=Modulator_0, children [Random_1], properties {pitchMax=1.1, pitchMin=0.8, volumeMax=1.3, volumeMin=1.1}}
 {id=QualityLevel_0, children [Modulator_0,Modulator_0,Modulator_0,Modulator_0]}
 {
  id=Random_1
  children [WavePlayer_0, WavePlayer_1, WavePlayer_2, WavePlayer_4, WavePlayer_9, WavePlayer_10]
  properties {weights [1,1,1,1,1,1]}
 }
 {id=WavePlayer_0, properties {soundWaveAssetPtr=&/Spot_Monkey_Gibbon_Call_01}}
 {id=WavePlayer_1, properties {soundWaveAssetPtr=&/Spot_Monkey_Gibbon_Call_02}}
 {id=WavePlayer_10, properties {soundWaveAssetPtr=&/Spot_Monkey_Gibbon_Call_06}}
 {id=WavePlayer_2, properties {soundWaveAssetPtr=&/Spot_Monkey_Gibbon_Call_03}}
 {id=WavePlayer_4, properties {soundWaveAssetPtr=&/Spot_Monkey_Gibbon_Call_04}}
 {id=WavePlayer_9, properties {soundWaveAssetPtr=&/Spot_Monkey_Gibbon_Call_05}}
]
properties {
 attenuationSettings=&Fort_TileAmbient_OneShot_Bird_Att
 maxDistance=8005
 soundClassObject=&Fort_Ambient_OneShots_Outdoors
 volumeMultiplier=0.501
 busSends [{
   audioBus=null
   maxSendDistance=5000
   maxSendLevel=1.2
   minSendDistance=3000
   minSendLevel=0.2
   sendLevel=1
   soundSourceBus=&AmbientAudio_Exterior_SourceBus
  }]
 preEffectBusSends [{}]
}
