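# Sound cue for daytime rainforest bird spot one-shots, routed to the indoor ambient one-shot sound class at low volume and lowered pitch.
# A top Random node picks one of six equally weighted bird groups: singing, sparrow, cormorant, Sonoma song, starling or alien bird.
# Each group randomly picks a wave through a pitch and volume Modulator, with alien birds set quieter.
# All quality levels share this tree, and bird attenuation and rainforest concurrency apply.
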SoundCue /Game/Sounds/Asteria/Biomes/Rainforest/Spots/Spot_Rainforest_Birds_Day_All_Indoors_Cue
root=QualityLevel_0
nodes [
 {id=Modulator_0, children [Random_3], properties {pitchMax=1.1, volumeMax=1.1, volumeMin=0.9}}
 {id=Modulator_1, children [Random_0], properties {pitchMax=1.1, volumeMax=1.1, volumeMin=0.9}}
 {id=Modulator_3, children [Random_4], properties {pitchMax=1.1, volumeMax=1.1, volumeMin=0.9}}
 {id=Modulator_4, children [Random_5], properties {pitchMax=1.1, volumeMax=1.1, volumeMin=0.9}}
 {id=Modulator_5, children [Random_6], properties {pitchMax=1.1, volumeMax=1.1, volumeMin=0.9}}
 {id=Modulator_6, children [Random_7], properties {pitchMax=1.1, volumeMax=0.9, volumeMin=0.8}}
 {id=QualityLevel_0, children [Random_2,Random_2,Random_2,Random_2]}
 {
  id=Random_0
  children [WavePlayer_6, WavePlayer_7, WavePlayer_8, WavePlayer_9, WavePlayer_10, WavePlayer_11]
  properties {weights [1,1,1,1,1,1]}
 }
 {
  id=Random_2
  children [Modulator_1,Modulator_0,Modulator_3,Modulator_4,Modulator_5,Modulator_6]
  properties {bShouldExcludeFromBranchCulling=true, weights [1,1,1,1,1,1]}
 }
 {id=Random_3, children [WavePlayer_0,WavePlayer_1,WavePlayer_2,WavePlayer_3,WavePlayer_4], properties {weights [1,1,1,1,1]}}
 {
  id=Random_4
  children [WavePlayer_5, WavePlayer_15, WavePlayer_16, WavePlayer_17, WavePlayer_18, WavePlayer_19]
  properties {weights [1,1,1,1,1,1]}
 }
 {id=Random_5, children [WavePlayer_20, WavePlayer_21, WavePlayer_22, WavePlayer_23], properties {weights [1,1,1,1]}}
 {
  id=Random_6
  children [WavePlayer_24, WavePlayer_25, WavePlayer_26, WavePlayer_27, WavePlayer_28, WavePlayer_29, WavePlayer_30, WavePlayer_31, WavePlayer_32]
  properties {weights [1,1,1,1,1,1,1,1,1]}
 }
 {
  id=Random_7
  children [
   WavePlayer_33
   WavePlayer_14
   WavePlayer_34
   WavePlayer_35
   WavePlayer_36
   WavePlayer_37
   WavePlayer_38
   WavePlayer_39
   WavePlayer_40
   WavePlayer_41
  ]
  properties {weights [1,1,1,1,1,1,1,1,1,1]}
 }
 {
  id=WavePlayer_0
  properties {soundWaveAssetPtr=/Game/Sounds/Artemis/Biomes/Tropical/Tropical_Spots/Spot_Tropical_Day_Bird_African_Sparrow_03}
 }
 {
  id=WavePlayer_1
  properties {soundWaveAssetPtr=/Game/Sounds/Artemis/Biomes/Tropical/Tropical_Spots/Spot_Tropical_Day_Bird_African_Sparrow_05}
 }
 {id=WavePlayer_10, properties {soundWaveAssetPtr=/Game/Sounds/Artemis/Biomes/Tropical/Tropical_Spots/Spot_Tropical_Day_Bird_Singing_05}}
 {id=WavePlayer_11, properties {soundWaveAssetPtr=/Game/Sounds/Artemis/Biomes/Tropical/Tropical_Spots/Spot_Tropical_Day_Bird_Singing_06}}
 {id=WavePlayer_14, properties {soundWaveAssetPtr=/Game/Sounds/Artemis/Biomes/Reality_Forest/Spots/AMB_Biome_Reality_Forest_SS_AlienBird_02}}
 {
  id=WavePlayer_15
  properties {soundWaveAssetPtr=/Game/Sounds/Artemis/Biomes/Tropical/Tropical_Spots/Spot_Tropical_Day_Bird_Cormorant_Tweets_02}
 }
 {
  id=WavePlayer_16
  properties {soundWaveAssetPtr=/Game/Sounds/Artemis/Biomes/Tropical/Tropical_Spots/Spot_Tropical_Day_Bird_Cormorant_Tweets_03}
 }
 {
  id=WavePlayer_17
  properties {soundWaveAssetPtr=/Game/Sounds/Artemis/Biomes/Tropical/Tropical_Spots/Spot_Tropical_Day_Bird_Cormorant_Tweets_04}
 }
 {
  id=WavePlayer_18
  properties {soundWaveAssetPtr=/Game/Sounds/Artemis/Biomes/Tropical/Tropical_Spots/Spot_Tropical_Day_Bird_Cormorant_Tweets_05}
 }
 {
  id=WavePlayer_19
  properties {soundWaveAssetPtr=/Game/Sounds/Artemis/Biomes/Tropical/Tropical_Spots/Spot_Tropical_Day_Bird_Cormorant_Tweets_06}
 }
 {
  id=WavePlayer_2
  properties {soundWaveAssetPtr=/Game/Sounds/Artemis/Biomes/Tropical/Tropical_Spots/Spot_Tropical_Day_Bird_African_Sparrow_01}
 }
 {id=WavePlayer_20, properties {soundWaveAssetPtr=/Game/Sounds/Artemis/Biomes/Tropical/Tropical_Spots/Spot_Tropical_Day_Bird_Sonoma_Song_01}}
 {id=WavePlayer_21, properties {soundWaveAssetPtr=/Game/Sounds/Artemis/Biomes/Tropical/Tropical_Spots/Spot_Tropical_Day_Bird_Sonoma_Song_02}}
 {id=WavePlayer_22, properties {soundWaveAssetPtr=/Game/Sounds/Artemis/Biomes/Tropical/Tropical_Spots/Spot_Tropical_Day_Bird_Sonoma_Song_03}}
 {id=WavePlayer_23, properties {soundWaveAssetPtr=/Game/Sounds/Artemis/Biomes/Tropical/Tropical_Spots/Spot_Tropical_Day_Bird_Sonoma_Song_04}}
 {id=WavePlayer_24, properties {soundWaveAssetPtr=/Game/Sounds/Artemis/Biomes/Tropical/Tropical_Spots/Spot_Tropical_Day_Egyptian_Starling_01}}
 {id=WavePlayer_25, properties {soundWaveAssetPtr=/Game/Sounds/Artemis/Biomes/Tropical/Tropical_Spots/Spot_Tropical_Day_Egyptian_Starling_02}}
 {id=WavePlayer_26, properties {soundWaveAssetPtr=/Game/Sounds/Artemis/Biomes/Tropical/Tropical_Spots/Spot_Tropical_Day_Egyptian_Starling_03}}
 {id=WavePlayer_27, properties {soundWaveAssetPtr=/Game/Sounds/Artemis/Biomes/Tropical/Tropical_Spots/Spot_Tropical_Day_Egyptian_Starling_04}}
 {id=WavePlayer_28, properties {soundWaveAssetPtr=/Game/Sounds/Artemis/Biomes/Tropical/Tropical_Spots/Spot_Tropical_Day_Egyptian_Starling_05}}
 {id=WavePlayer_29, properties {soundWaveAssetPtr=/Game/Sounds/Artemis/Biomes/Tropical/Tropical_Spots/Spot_Tropical_Day_Egyptian_Starling_06}}
 {
  id=WavePlayer_3
  properties {soundWaveAssetPtr=/Game/Sounds/Artemis/Biomes/Tropical/Tropical_Spots/Spot_Tropical_Day_Bird_African_Sparrow_02}
 }
 {id=WavePlayer_30, properties {soundWaveAssetPtr=/Game/Sounds/Artemis/Biomes/Tropical/Tropical_Spots/Spot_Tropical_Day_Egyptian_Starling_07}}
 {id=WavePlayer_31, properties {soundWaveAssetPtr=/Game/Sounds/Artemis/Biomes/Tropical/Tropical_Spots/Spot_Tropical_Day_Egyptian_Starling_08}}
 {id=WavePlayer_32, properties {soundWaveAssetPtr=/Game/Sounds/Artemis/Biomes/Tropical/Tropical_Spots/Spot_Tropical_Day_Egyptian_Starling_09}}
 {id=WavePlayer_33, properties {soundWaveAssetPtr=/Game/Sounds/Artemis/Biomes/Reality_Forest/Spots/AMB_Biome_Reality_Forest_SS_AlienBird_01}}
 {id=WavePlayer_34, properties {soundWaveAssetPtr=/Game/Sounds/Artemis/Biomes/Reality_Forest/Spots/AMB_Biome_Reality_Forest_SS_AlienBird_03}}
 {id=WavePlayer_35, properties {soundWaveAssetPtr=/Game/Sounds/Artemis/Biomes/Reality_Forest/Spots/AMB_Biome_Reality_Forest_SS_AlienBird_04}}
 {id=WavePlayer_36, properties {soundWaveAssetPtr=/Game/Sounds/Artemis/Biomes/Reality_Forest/Spots/AMB_Biome_Reality_Forest_SS_AlienBird_05}}
 {id=WavePlayer_37, properties {soundWaveAssetPtr=/Game/Sounds/Artemis/Biomes/Reality_Forest/Spots/AMB_Biome_Reality_Forest_SS_AlienBird_06}}
 {id=WavePlayer_38, properties {soundWaveAssetPtr=/Game/Sounds/Artemis/Biomes/Reality_Forest/Spots/AMB_Biome_Reality_Forest_SS_AlienBird_07}}
 {id=WavePlayer_39, properties {soundWaveAssetPtr=/Game/Sounds/Artemis/Biomes/Reality_Forest/Spots/AMB_Biome_Reality_Forest_SS_AlienBird_08}}
 {
  id=WavePlayer_4
  properties {soundWaveAssetPtr=/Game/Sounds/Artemis/Biomes/Tropical/Tropical_Spots/Spot_Tropical_Day_Bird_African_Sparrow_04}
 }
 {id=WavePlayer_40, properties {soundWaveAssetPtr=/Game/Sounds/Artemis/Biomes/Reality_Forest/Spots/AMB_Biome_Reality_Forest_SS_AlienBird_09}}
 {id=WavePlayer_41, properties {soundWaveAssetPtr=/Game/Sounds/Artemis/Biomes/Reality_Forest/Spots/AMB_Biome_Reality_Forest_SS_AlienBird_10}}
 {
  id=WavePlayer_5
  properties {soundWaveAssetPtr=/Game/Sounds/Artemis/Biomes/Tropical/Tropical_Spots/Spot_Tropical_Day_Bird_Cormorant_Tweets_01}
 }
 {id=WavePlayer_6, properties {soundWaveAssetPtr=/Game/Sounds/Artemis/Biomes/Tropical/Tropical_Spots/Spot_Tropical_Day_Bird_Singing_01}}
 {id=WavePlayer_7, properties {soundWaveAssetPtr=/Game/Sounds/Artemis/Biomes/Tropical/Tropical_Spots/Spot_Tropical_Day_Bird_Singing_02}}
 {id=WavePlayer_8, properties {soundWaveAssetPtr=/Game/Sounds/Artemis/Biomes/Tropical/Tropical_Spots/Spot_Tropical_Day_Bird_Singing_03}}
 {id=WavePlayer_9, properties {soundWaveAssetPtr=/Game/Sounds/Artemis/Biomes/Tropical/Tropical_Spots/Spot_Tropical_Day_Bird_Singing_04}}
]
properties {
 attenuationSettings=&Fort_TileAmbient_OneShot_Bird_Att
 bEnableBusSends=false
 maxDistance=8005
 pitchMultiplier=0.85
 soundClassObject=/Game/Sounds/FortSoundClassesAndModes/Fort_Ambient_OneShots_Indoors
 volumeMultiplier=0.1
 concurrencyOverrides {maxCount=2}
 concurrencySet [/Game/Sounds/Concurrency_Rules/Asteria_Rainforest_Spots_Conc]
 preEffectBusSends [{}]
}
