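# Sound cue for nighttime rainforest bird one-shot ambience, covering hawk, crane, crow and SteelPecker calls and croaks.
# A top random node picks one of five equally weighted call groups, each passing through its own pitch and volume modulator into a random choice of wave variations.
# All four quality-level inputs share that same random node.
# Playback uses bird one-shot attenuation, an outdoor ambient sound class and rainforest spot concurrency rules.
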SoundCue /Game/Sounds/Asteria/Biomes/Rainforest/Spots/Spot_Rainforest_Birds_Night_All_Cue
root=QualityLevel_0
nodes [
 {id=Modulator_0, children [Random_2], properties {pitchMax=1.1, volumeMax=0.6, volumeMin=0.4}}
 {id=Modulator_1, children [Random_1], properties {pitchMax=0.9, pitchMin=0.7, volumeMax=1, volumeMin=0.8}}
 {id=Modulator_2, children [Random_0], properties {pitchMax=1.1, pitchMin=0.8, volumeMax=2.2, volumeMin=2}}
 {id=Modulator_3, children [Random_4], properties {pitchMax=1, pitchMin=0.9, volumeMax=1, volumeMin=0.8}}
 {id=Modulator_5, children [Random_6], properties {pitchMax=1.1, pitchMin=0.8, volumeMax=1.2, volumeMin=1}}
 {id=QualityLevel_0, children [Random_3,Random_3,Random_3,Random_3]}
 {
  id=Random_0
  children [WavePlayer_10, WavePlayer_11, WavePlayer_12, WavePlayer_13, WavePlayer_14, WavePlayer_15, WavePlayer_16]
  properties {weights [1,1,1,1,1,1,1]}
 }
 {id=Random_1, children [WavePlayer_5,WavePlayer_6,WavePlayer_7,WavePlayer_8,WavePlayer_9], properties {weights [1,1,1,1,1]}}
 {id=Random_2, children [WavePlayer_0,WavePlayer_1,WavePlayer_2,WavePlayer_3,WavePlayer_4], properties {weights [1,1,1,1,1]}}
 {id=Random_3, children [Modulator_0,Modulator_1,Modulator_2,Modulator_3,Modulator_5], properties {weights [1,1,1,1,1]}}
 {
  id=Random_4
  children [WavePlayer_17, WavePlayer_18, WavePlayer_19, WavePlayer_20, WavePlayer_21, WavePlayer_22]
  properties {weights [1,1,1,1,1,1]}
 }
 {id=Random_6, children [WavePlayer_27, WavePlayer_28, WavePlayer_29, WavePlayer_30, WavePlayer_31], properties {weights [1,1,1,1,1]}}
 {id=WavePlayer_0, properties {soundWaveAssetPtr=/Game/Sounds/Artemis/Biomes/Tropical/Tropical_Spots/Spot_Tropical_Night_Hawk_Bird_Call_05}}
 {id=WavePlayer_1, properties {soundWaveAssetPtr=/Game/Sounds/Artemis/Biomes/Tropical/Tropical_Spots/Spot_Tropical_Night_Hawk_Bird_Call_01}}
 {id=WavePlayer_10, properties {soundWaveAssetPtr=&/Spot_Bird_Rainforest_Crane_01}}
 {id=WavePlayer_11, properties {soundWaveAssetPtr=&/Spot_Bird_Rainforest_Crane_02}}
 {id=WavePlayer_12, properties {soundWaveAssetPtr=&/Spot_Bird_Rainforest_Crane_03}}
 {id=WavePlayer_13, properties {soundWaveAssetPtr=&/Spot_Bird_Rainforest_Crane_04}}
 {id=WavePlayer_14, properties {soundWaveAssetPtr=&/Spot_Bird_Rainforest_Crane_05}}
 {id=WavePlayer_15, properties {soundWaveAssetPtr=&/Spot_Bird_Rainforest_Crane_06}}
 {id=WavePlayer_16, properties {soundWaveAssetPtr=&/Spot_Bird_Rainforest_Crane_07}}
 {id=WavePlayer_17, properties {soundWaveAssetPtr=&/Spot_Bird_Tropical_Crow_Call_01}}
 {id=WavePlayer_18, properties {soundWaveAssetPtr=&/Spot_Bird_Tropical_Crow_Call_02}}
 {id=WavePlayer_19, properties {soundWaveAssetPtr=&/Spot_Bird_Tropical_Crow_Call_03}}
 {id=WavePlayer_2, properties {soundWaveAssetPtr=/Game/Sounds/Artemis/Biomes/Tropical/Tropical_Spots/Spot_Tropical_Night_Hawk_Bird_Call_02}}
 {id=WavePlayer_20, properties {soundWaveAssetPtr=&/Spot_Bird_Tropical_Crow_Call_04}}
 {id=WavePlayer_21, properties {soundWaveAssetPtr=&/Spot_Bird_Tropical_Crow_Call_05}}
 {id=WavePlayer_22, properties {soundWaveAssetPtr=&/Spot_Bird_Tropical_Crow_Call_06}}
 {id=WavePlayer_27, properties {soundWaveAssetPtr=&/Spot_Bird_Tropical_Night_SteelPecker_Croak_03}}
 {id=WavePlayer_28, properties {soundWaveAssetPtr=&/Spot_Bird_Tropical_Night_SteelPecker_Croak_05}}
 {id=WavePlayer_29, properties {soundWaveAssetPtr=&/Spot_Bird_Tropical_Night_SteelPecker_Croak_01}}
 {id=WavePlayer_3, properties {soundWaveAssetPtr=/Game/Sounds/Artemis/Biomes/Tropical/Tropical_Spots/Spot_Tropical_Night_Hawk_Bird_Call_03}}
 {id=WavePlayer_30, properties {soundWaveAssetPtr=&/Spot_Bird_Tropical_Night_SteelPecker_Croak_02}}
 {id=WavePlayer_31, properties {soundWaveAssetPtr=&/Spot_Bird_Tropical_Night_SteelPecker_Croak_04}}
 {id=WavePlayer_4, properties {soundWaveAssetPtr=/Game/Sounds/Artemis/Biomes/Tropical/Tropical_Spots/Spot_Tropical_Night_Hawk_Bird_Call_04}}
 {id=WavePlayer_5, properties {soundWaveAssetPtr=&/Spot_Bird_Night_SteelPecker_Calls_01}}
 {id=WavePlayer_6, properties {soundWaveAssetPtr=&/Spot_Bird_Night_SteelPecker_Calls_02}}
 {id=WavePlayer_7, properties {soundWaveAssetPtr=&/Spot_Bird_Night_SteelPecker_Calls_03}}
 {id=WavePlayer_8, properties {soundWaveAssetPtr=&/Spot_Bird_Night_SteelPecker_Calls_04}}
 {id=WavePlayer_9, properties {soundWaveAssetPtr=&/Spot_Bird_Night_SteelPecker_Calls_05}}
]
properties {
 attenuationSettings=&Fort_TileAmbient_OneShot_Bird_Att
 bEnableBusSends=false
 maxDistance=8005
 soundClassObject=&Fort_Ambient_OneShots_Outdoors
 volumeMultiplier=0.19
 concurrencySet [/Game/Sounds/Concurrency_Rules/Asteria_Rainforest_Spots_Conc]
 preEffectBusSends [{}]
}
