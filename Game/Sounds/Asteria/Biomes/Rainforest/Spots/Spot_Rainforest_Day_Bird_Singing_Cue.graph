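# Sound cue for a daytime rainforest bird-singing ambient one-shot, reusing the Artemis Tropical bird recordings.
# A quality-level root sends its first two slots to a random pick of six waves and its last two to a three-wave subset.
# Each branch passes through a modulator for slight pitch and volume variation.
# Output uses bird one-shot attenuation, the outdoor ambient sound class and a send to the exterior ambient bus.
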SoundCue /Game/Sounds/Asteria/Biomes/Rainforest/Spots/Spot_Rainforest_Day_Bird_Singing_Cue
root=QualityLevel_0
nodes [
 {id=Modulator_1, children [Random_0], properties {pitchMax=1.1, volumeMax=1.1, volumeMin=0.9}}
 {id=Modulator_2, children [Random_1], properties {pitchMax=1.1, volumeMax=1.1, volumeMin=0.9}}
 {id=QualityLevel_0, children [Modulator_1,Modulator_1,Modulator_2,Modulator_2]}
 {
  id=Random_0
  children [WavePlayer_6, WavePlayer_7, WavePlayer_8, WavePlayer_9, WavePlayer_10, WavePlayer_11]
  properties {weights [1,1,1,1,1,1]}
 }
 {id=Random_1, children [WavePlayer_12, WavePlayer_13, WavePlayer_14], properties {weights [1,1,1]}}
 {id=WavePlayer_10, properties {soundWaveAssetPtr=/Game/Sounds/Artemis/Biomes/Tropical/Tropical_Spots/Spot_Tropical_Day_Bird_Singing_05}}
 {id=WavePlayer_11, properties {soundWaveAssetPtr=/Game/Sounds/Artemis/Biomes/Tropical/Tropical_Spots/Spot_Tropical_Day_Bird_Singing_06}}
 {id=WavePlayer_12, properties {soundWaveAssetPtr=&Spot_Tropical_Day_Bird_Singing_05}}
 {id=WavePlayer_13, properties {soundWaveAssetPtr=&Spot_Tropical_Day_Bird_Singing_06}}
 {id=WavePlayer_14, properties {soundWaveAssetPtr=/Game/Sounds/Artemis/Biomes/Tropical/Tropical_Spots/Spot_Tropical_Day_Bird_Singing_01}}
 {id=WavePlayer_6, properties {soundWaveAssetPtr=&Spot_Tropical_Day_Bird_Singing_01}}
 {id=WavePlayer_7, properties {soundWaveAssetPtr=/Game/Sounds/Artemis/Biomes/Tropical/Tropical_Spots/Spot_Tropical_Day_Bird_Singing_02}}
 {id=WavePlayer_8, properties {soundWaveAssetPtr=/Game/Sounds/Artemis/Biomes/Tropical/Tropical_Spots/Spot_Tropical_Day_Bird_Singing_03}}
 {id=WavePlayer_9, properties {soundWaveAssetPtr=/Game/Sounds/Artemis/Biomes/Tropical/Tropical_Spots/Spot_Tropical_Day_Bird_Singing_04}}
]
properties {
 attenuationSettings=&Fort_TileAmbient_OneShot_Bird_Att
 maxDistance=8005
 soundClassObject=&Fort_Ambient_OneShots_Outdoors
 volumeMultiplier=0.301
 busSends [{audioBus=null, minSendLevel=0.6, sendLevel=1, soundSourceBus=&AmbientAudio_Exterior_SourceBus}]
 preEffectBusSends [{}]
}
