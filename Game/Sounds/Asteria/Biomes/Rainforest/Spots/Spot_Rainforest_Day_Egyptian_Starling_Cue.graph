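# Sound cue for daytime Egyptian starling bird spot ambience in the Asteria rainforest biome, reusing Artemis tropical recordings.
# A quality-level node routes its first two slots to a random pick of nine waves and its last two to a four-wave subset.
# Each random set feeds a modulator for slight pitch and volume variation.
# Playback uses bird one-shot attenuation and the outdoor ambient one-shot sound class at reduced volume.
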SoundCue /Game/Sounds/Asteria/Biomes/Rainforest/Spots/Spot_Rainforest_Day_Egyptian_Starling_Cue
root=QualityLevel_0
nodes [
 {id=Modulator_1, children [Random_2], properties {pitchMax=1.1, volumeMax=1.1, volumeMin=0.9}}
 {id=Modulator_2, children [Random_3], properties {pitchMax=1.1, volumeMax=1.1, volumeMin=0.9}}
 {id=QualityLevel_0, children [Modulator_1,Modulator_1,Modulator_2,Modulator_2]}
 {
  id=Random_2
  children [WavePlayer_0,WavePlayer_1,WavePlayer_2,WavePlayer_3,WavePlayer_4,WavePlayer_5,WavePlayer_6,WavePlayer_8,WavePlayer_9]
  properties {weights [1,1,1,1,1,1,1,1,1]}
 }
 {id=Random_3, children [WavePlayer_10, WavePlayer_11, WavePlayer_12, WavePlayer_13], properties {weights [1,1,1,1]}}
 {id=WavePlayer_0, properties {soundWaveAssetPtr=/Game/Sounds/Artemis/Biomes/Tropical/Tropical_Spots/Spot_Tropical_Day_Egyptian_Starling_01}}
 {id=WavePlayer_1, properties {soundWaveAssetPtr=/Game/Sounds/Artemis/Biomes/Tropical/Tropical_Spots/Spot_Tropical_Day_Egyptian_Starling_02}}
 {id=WavePlayer_10, properties {soundWaveAssetPtr=/Game/Sounds/Artemis/Biomes/Tropical/Tropical_Spots/Spot_Tropical_Day_Egyptian_Starling_08}}
 {id=WavePlayer_11, properties {soundWaveAssetPtr=/Game/Sounds/Artemis/Biomes/Tropical/Tropical_Spots/Spot_Tropical_Day_Egyptian_Starling_09}}
 {id=WavePlayer_12, properties {soundWaveAssetPtr=/Game/Sounds/Artemis/Biomes/Tropical/Tropical_Spots/Spot_Tropical_Day_Egyptian_Starling_03}}
 {id=WavePlayer_13, properties {soundWaveAssetPtr=/Game/Sounds/Artemis/Biomes/Tropical/Tropical_Spots/Spot_Tropical_Day_Egyptian_Starling_05}}
 {id=WavePlayer_2, properties {soundWaveAssetPtr=&Spot_Tropical_Day_Egyptian_Starling_03}}
 {id=WavePlayer_3, properties {soundWaveAssetPtr=/Game/Sounds/Artemis/Biomes/Tropical/Tropical_Spots/Spot_Tropical_Day_Egyptian_Starling_04}}
 {id=WavePlayer_4, properties {soundWaveAssetPtr=&Spot_Tropical_Day_Egyptian_Starling_05}}
 {id=WavePlayer_5, properties {soundWaveAssetPtr=/Game/Sounds/Artemis/Biomes/Tropical/Tropical_Spots/Spot_Tropical_Day_Egyptian_Starling_06}}
 {id=WavePlayer_6, properties {soundWaveAssetPtr=/Game/Sounds/Artemis/Biomes/Tropical/Tropical_Spots/Spot_Tropical_Day_Egyptian_Starling_07}}
 {id=WavePlayer_8, properties {soundWaveAssetPtr=&Spot_Tropical_Day_Egyptian_Starling_08}}
 {id=WavePlayer_9, properties {soundWaveAssetPtr=&Spot_Tropical_Day_Egyptian_Starling_09}}
]
properties {
 attenuationSettings=&Fort_TileAmbient_OneShot_Bird_Att
 bEnableBusSends=false
 maxDistance=8005
 soundClassObject=&Fort_Ambient_OneShots_Outdoors
 volumeMultiplier=0.251
 preEffectBusSends [{}]
}
