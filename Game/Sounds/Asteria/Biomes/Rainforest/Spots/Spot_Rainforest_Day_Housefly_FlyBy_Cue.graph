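# Sound cue for housefly fly-by one-shots in the daytime Rainforest biome ambience, using the insect fly-by attenuation and ambient one-shot sound class.
# A quality-level switch sends its first two slots to a random pick of seven waves and its last two to a random pick of three.
# Each branch randomizes pitch and volume through a modulator.
# All waves are reused from the Artemis Tropical spot set.
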SoundCue /Game/Sounds/Asteria/Biomes/Rainforest/Spots/Spot_Rainforest_Day_Housefly_FlyBy_Cue
root=QualityLevel_0
nodes [
 {id=Modulator_1, children [Random_0], properties {pitchMax=1.1, volumeMax=1.1, volumeMin=0.9}}
 {id=Modulator_2, children [Random_1], properties {pitchMax=1.1, volumeMax=1.1, volumeMin=0.9}}
 {id=QualityLevel_0, children [Modulator_1,Modulator_1,Modulator_2,Modulator_2]}
 {
  id=Random_0
  children [WavePlayer_7, WavePlayer_16, WavePlayer_17, WavePlayer_18, WavePlayer_22, WavePlayer_23, WavePlayer_30]
  properties {weights [1,1,1,1,1,1,1]}
 }
 {id=Random_1, children [WavePlayer_31, WavePlayer_32, WavePlayer_33], properties {weights [1,1,1]}}
 {id=WavePlayer_16, properties {soundWaveAssetPtr=/Game/Sounds/Artemis/Biomes/Tropical/Tropical_Spots/Spot_Tropical_Day_HouseFly_Flyby_02}}
 {id=WavePlayer_17, properties {soundWaveAssetPtr=/Game/Sounds/Artemis/Biomes/Tropical/Tropical_Spots/Spot_Tropical_Day_HouseFly_Flyby_03}}
 {id=WavePlayer_18, properties {soundWaveAssetPtr=/Game/Sounds/Artemis/Biomes/Tropical/Tropical_Spots/Spot_Tropical_Day_HouseFly_Flyby_04}}
 {id=WavePlayer_22, properties {soundWaveAssetPtr=/Game/Sounds/Artemis/Biomes/Tropical/Tropical_Spots/Spot_Tropical_Day_HouseFly_Flyby_05}}
 {id=WavePlayer_23, properties {soundWaveAssetPtr=/Game/Sounds/Artemis/Biomes/Tropical/Tropical_Spots/Spot_Tropical_Day_HouseFly_Flyby_06}}
 {id=WavePlayer_30, properties {soundWaveAssetPtr=/Game/Sounds/Artemis/Biomes/Tropical/Tropical_Spots/Spot_Tropical_Day_HouseFly_Flyby_07}}
 {id=WavePlayer_31, properties {soundWaveAssetPtr=&Spot_Tropical_Day_HouseFly_Flyby_04}}
 {id=WavePlayer_32, properties {soundWaveAssetPtr=/Game/Sounds/Artemis/Biomes/Tropical/Tropical_Spots/Spot_Tropical_Day_HouseFly_Flyby_01}}
 {id=WavePlayer_33, properties {soundWaveAssetPtr=&Spot_Tropical_Day_HouseFly_Flyby_06}}
 {id=WavePlayer_7, properties {soundWaveAssetPtr=&Spot_Tropical_Day_HouseFly_Flyby_01}}
]
properties {
 attenuationSettings=/Game/Sounds/Attenuation/Fort_TileAmbient_OneShot_Insect_By_Att
 bEnableBusSends=false
 maxDistance=6000
 soundClassObject=&Fort_Ambient_OneShots
 volumeMultiplier=0.376
 preEffectBusSends [{}]
}
