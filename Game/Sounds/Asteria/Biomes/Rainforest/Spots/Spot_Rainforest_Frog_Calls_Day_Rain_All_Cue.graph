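# Sound cue for one-shot frog call ambience in the Rainforest biome, named for daytime and rain.
# An equal-weight random node, excluded from branch culling and shared by every quality level, picks one of five random wave pools.
# Striped tree, spring pepper and mating call pools get pitch and volume modulation; borrowed Apollo swamp and Eastern frog-insect pools play unmodulated.
# Output uses outdoor one-shot sound class and bird one-shot attenuation at reduced volume.
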SoundCue /Game/Sounds/Asteria/Biomes/Rainforest/Spots/Spot_Rainforest_Frog_Calls_Day_Rain_All_Cue
root=QualityLevel_0
nodes [
 {id=Modulator_2, children [Random_3], properties {pitchMax=1.1, pitchMin=0.8, volumeMax=1.2, volumeMin=1}}
 {id=Modulator_3, children [Random_4], properties {pitchMax=1.1, pitchMin=0.8, volumeMax=1.2, volumeMin=1}}
 {id=Modulator_4, children [Random_5], properties {pitchMax=1, pitchMin=0.8, volumeMax=1.2, volumeMin=1}}
 {id=QualityLevel_0, children [Random_2,Random_2,Random_2,Random_2]}
 {
  id=Random_1
  children [
   WavePlayer_4
   WavePlayer_5
   WavePlayer_6
   WavePlayer_7
   WavePlayer_8
   WavePlayer_9
   WavePlayer_10
   WavePlayer_12
   WavePlayer_36
   WavePlayer_37
  ]
  properties {weights [1,1,1,1,1,1,1,1,1,1]}
 }
 {
  id=Random_2
  children [Modulator_2,Modulator_3,Modulator_4,Random_1,Random_6]
  properties {bShouldExcludeFromBranchCulling=true, weights [1,1,1,1,1]}
 }
 {
  id=Random_3
  children [WavePlayer_11, WavePlayer_13, WavePlayer_14, WavePlayer_15, WavePlayer_16, WavePlayer_17, WavePlayer_18]
  properties {weights [1,1,1,1,1,1,1]}
 }
 {id=Random_4, children [WavePlayer_19, WavePlayer_20, WavePlayer_21, WavePlayer_22, WavePlayer_23], properties {weights [1,1,1,1,1]}}
 {
  id=Random_5
  children [
   WavePlayer_24
   WavePlayer_25
   WavePlayer_26
   WavePlayer_27
   WavePlayer_28
   WavePlayer_29
   WavePlayer_30
   WavePlayer_31
   WavePlayer_32
   WavePlayer_33
   WavePlayer_34
   WavePlayer_35
  ]
  properties {weights [1,1,1,1,1,1,1,1,1,1,1,1]}
 }
 {
  id=Random_6
  children [
   WavePlayer_38
   WavePlayer_39
   WavePlayer_40
   WavePlayer_41
   WavePlayer_42
   WavePlayer_43
   WavePlayer_44
   WavePlayer_45
   WavePlayer_46
   WavePlayer_47
  ]
  properties {weights [1,1,1,1,1,1,1,1,1,1]}
 }
 {id=WavePlayer_10, properties {soundWaveAssetPtr=/Game/Sounds/Apollo/Biomes/Swamp_01/Biome_Swamp_01_CricketFrog_03}}
 {id=WavePlayer_11, properties {soundWaveAssetPtr=&/Spot_Frog_Striped_Tree_Calls_01}}
 {id=WavePlayer_12, properties {soundWaveAssetPtr=/Game/Sounds/Apollo/Biomes/Swamp_01/Biome_Swamp_01_Frog_01}}
 {id=WavePlayer_13, properties {soundWaveAssetPtr=&/Spot_Frog_Striped_Tree_Calls_02}}
 {id=WavePlayer_14, properties {soundWaveAssetPtr=&/Spot_Frog_Striped_Tree_Calls_03}}
 {id=WavePlayer_15, properties {soundWaveAssetPtr=&/Spot_Frog_Striped_Tree_Calls_04}}
 {id=WavePlayer_16, properties {soundWaveAssetPtr=&/Spot_Frog_Striped_Tree_Calls_05}}
 {id=WavePlayer_17, properties {soundWaveAssetPtr=&/Spot_Frog_Striped_Tree_Calls_06}}
 {id=WavePlayer_18, properties {soundWaveAssetPtr=&/Spot_Frog_Striped_Tree_Calls_07}}
 {id=WavePlayer_19, properties {soundWaveAssetPtr=&/Spot_Frog_Spring_Pepper_Call_01}}
 {id=WavePlayer_20, properties {soundWaveAssetPtr=&/Spot_Frog_Spring_Pepper_Call_02}}
 {id=WavePlayer_21, properties {soundWaveAssetPtr=&/Spot_Frog_Spring_Pepper_Call_03}}
 {id=WavePlayer_22, properties {soundWaveAssetPtr=&/Spot_Frog_Spring_Pepper_Call_04}}
 {id=WavePlayer_23, properties {soundWaveAssetPtr=&/Spot_Frog_Spring_Pepper_Call_05}}
 {id=WavePlayer_24, properties {soundWaveAssetPtr=&/Spot_Frog_Mating_Call_01}}
 {id=WavePlayer_25, properties {soundWaveAssetPtr=&/Spot_Frog_Mating_Call_02}}
 {id=WavePlayer_26, properties {soundWaveAssetPtr=&/Spot_Frog_Mating_Call_03}}
 {id=WavePlayer_27, properties {soundWaveAssetPtr=&/Spot_Frog_Mating_Call_05}}
 {id=WavePlayer_28, properties {soundWaveAssetPtr=&/Spot_Frog_Mating_Call_06}}
 {id=WavePlayer_29, properties {soundWaveAssetPtr=&/Spot_Frog_Mating_Call_Pitched_01}}
 {id=WavePlayer_30, properties {soundWaveAssetPtr=&/Spot_Frog_Mating_Call_Pitched_02}}
 {id=WavePlayer_31, properties {soundWaveAssetPtr=&/Spot_Frog_Mating_Call_Pitched_03}}
 {id=WavePlayer_32, properties {soundWaveAssetPtr=&/Spot_Frog_Mating_Call_Pitched_04}}
 {id=WavePlayer_33, properties {soundWaveAssetPtr=&/Spot_Frog_Mating_Call_Pitched_05}}
 {id=WavePlayer_34, properties {soundWaveAssetPtr=&/Spot_Frog_Mating_Call_Pitched_06}}
 {id=WavePlayer_35, properties {soundWaveAssetPtr=&/Spot_Frog_Mating_Call_04}}
 {id=WavePlayer_36, properties {soundWaveAssetPtr=/Game/Sounds/Apollo/Biomes/Swamp_01/Biome_Swamp_01_Frog_02}}
 {id=WavePlayer_37, properties {soundWaveAssetPtr=/Game/Sounds/Apollo/Biomes/Swamp_01/Biome_Swamp_01_Frog_04}}
 {id=WavePlayer_38, properties {soundWaveAssetPtr=/Game/Sounds/Asteria/Biomes/Eastern/Spots/Spot_Frog_Insects_Call_01}}
 {id=WavePlayer_39, properties {soundWaveAssetPtr=/Game/Sounds/Asteria/Biomes/Eastern/Spots/Spot_Frog_Insects_Call_05}}
 {id=WavePlayer_4, properties {soundWaveAssetPtr=/Game/Sounds/Apollo/Biomes/Swamp_01/Biome_Swamp_01_CricketFrog_01}}
 {id=WavePlayer_40, properties {soundWaveAssetPtr=/Game/Sounds/Asteria/Biomes/Eastern/Spots/Spot_Frog_Insects_Call_10}}
 {id=WavePlayer_41, properties {soundWaveAssetPtr=/Game/Sounds/Asteria/Biomes/Eastern/Spots/Spot_Frog_Insects_Call_02}}
 {id=WavePlayer_42, properties {soundWaveAssetPtr=/Game/Sounds/Asteria/Biomes/Eastern/Spots/Spot_Frog_Insects_Call_03}}
 {id=WavePlayer_43, properties {soundWaveAssetPtr=/Game/Sounds/Asteria/Biomes/Eastern/Spots/Spot_Frog_Insects_Call_04}}
 {id=WavePlayer_44, properties {soundWaveAssetPtr=/Game/Sounds/Asteria/Biomes/Eastern/Spots/Spot_Frog_Insects_Call_06}}
 {id=WavePlayer_45, properties {soundWaveAssetPtr=/Game/Sounds/Asteria/Biomes/Eastern/Spots/Spot_Frog_Insects_Call_07}}
 {id=WavePlayer_46, properties {soundWaveAssetPtr=/Game/Sounds/Asteria/Biomes/Eastern/Spots/Spot_Frog_Insects_Call_08}}
 {id=WavePlayer_47, properties {soundWaveAssetPtr=/Game/Sounds/Asteria/Biomes/Eastern/Spots/Spot_Frog_Insects_Call_09}}
 {id=WavePlayer_5, properties {soundWaveAssetPtr=/Game/Sounds/Apollo/Biomes/Swamp_01/Biome_Swamp_01_CricketFrog_04}}
 {id=WavePlayer_6, properties {soundWaveAssetPtr=/Game/Sounds/Apollo/Biomes/Swamp_01/Biome_Swamp_01_Frog_03}}
 {id=WavePlayer_7, properties {soundWaveAssetPtr=/Game/Sounds/Apollo/Biomes/Swamp_01/Biome_Swamp_01_Frog_05}}
 {id=WavePlayer_8, properties {soundWaveAssetPtr=/Game/Sounds/Apollo/Biomes/Swamp_01/Biome_Swamp_01_Frog_06}}
 {id=WavePlayer_9, properties {soundWaveAssetPtr=/Game/Sounds/Apollo/Biomes/Swamp_01/Biome_Swamp_01_CricketFrog_02}}
]
properties {
 attenuationSettings=&Fort_TileAmbient_OneShot_Bird_Att
 bEnableBusSends=false
 maxDistance=8005
 soundClassObject=&Fort_Ambient_OneShots_Outdoors
 volumeMultiplier=0.301
 preEffectBusSends [{}]
}
