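# Rainforest biome ambient spot cue that plays a random distant daytime monkey call, either gibbon or chimp.
# A top Random, exempt from branch culling, picks a species branch, where a Modulator varies pitch and volume over six equally weighted waves, with gibbons set louder.
# The root QualityLevel node feeds that Random to the first two levels and leaves the other two empty.
# Distant-creature attenuation, an outdoor one-shot sound class and rainforest spot concurrency govern playback.
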SoundCue /Game/Sounds/Asteria/Biomes/Rainforest/Spots/Spot_Rainforest_Monkey_Calls_Day_All_Cue
root=QualityLevel_0
nodes [
 {id=Modulator_0, children [Random_1], properties {pitchMax=1.1, pitchMin=0.8, volumeMax=1.3, volumeMin=1.1}}
 {id=Modulator_1, children [Random_0], properties {pitchMax=1.1, pitchMin=0.8, volumeMax=1, volumeMin=0.8}}
 {id=QualityLevel_0, children [Random_2,Random_2,null,null]}
 {
  id=Random_0
  children [WavePlayer_3, WavePlayer_5, WavePlayer_6, WavePlayer_7, WavePlayer_8, WavePlayer_12]
  properties {weights [1,1,1,1,1,1]}
 }
 {
  id=Random_1
  children [WavePlayer_0, WavePlayer_1, WavePlayer_2, WavePlayer_4, WavePlayer_9, WavePlayer_10]
  properties {weights [1,1,1,1,1,1]}
 }
 {id=Random_2, children [Modulator_0,Modulator_1], properties {bShouldExcludeFromBranchCulling=true, weights [1,1]}}
 {id=WavePlayer_0, properties {soundWaveAssetPtr=&/Spot_Monkey_Gibbon_Call_01}}
 {id=WavePlayer_1, properties {soundWaveAssetPtr=&/Spot_Monkey_Gibbon_Call_02}}
 {id=WavePlayer_10, properties {soundWaveAssetPtr=&/Spot_Monkey_Gibbon_Call_06}}
 {id=WavePlayer_12, properties {soundWaveAssetPtr=&/Spot_Monkey_Chimp_Call_06}}
 {id=WavePlayer_2, properties {soundWaveAssetPtr=&/Spot_Monkey_Gibbon_Call_03}}
 {id=WavePlayer_3, properties {soundWaveAssetPtr=&/Spot_Monkey_Chimp_Call_01}}
 {id=WavePlayer_4, properties {soundWaveAssetPtr=&/Spot_Monkey_Gibbon_Call_04}}
 {id=WavePlayer_5, properties {soundWaveAssetPtr=&/Spot_Monkey_Chimp_Call_02}}
 {id=WavePlayer_6, properties {soundWaveAssetPtr=&/Spot_Monkey_Chimp_Call_03}}
 {id=WavePlayer_7, properties {soundWaveAssetPtr=&/Spot_Monkey_Chimp_Call_04}}
 {id=WavePlayer_8, properties {soundWaveAssetPtr=&/Spot_Monkey_Chimp_Call_05}}
 {id=WavePlayer_9, properties {soundWaveAssetPtr=&/Spot_Monkey_Gibbon_Call_05}}
]
properties {
 attenuationSettings=/Game/Sounds/Attenuation/Fort_TileAmbient_OneShot_Distant_Creature_Att
 bEnableBusSends=false
 maxDistance=8500
 soundClassObject=&Fort_Ambient_OneShots_Outdoors
 volumeMultiplier=0.752
 concurrencySet [/Game/Sounds/Concurrency_Rules/Asteria_Rainforest_Spots_Conc]
 preEffectBusSends [{}]
}
