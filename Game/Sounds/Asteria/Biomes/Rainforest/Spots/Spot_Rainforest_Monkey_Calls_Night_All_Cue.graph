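# Sound cue for one-shot nighttime rainforest monkey calls, played as distant creature ambience with distance attenuation and a rainforest-spot concurrency rule.
# A random node picks evenly among howler calls, distant howler growls and gibbon calls, each choosing a random wave.
# Each group has its own pitch and volume modulation.
# The quality switch feeds that random node, exempt from branch culling, into its first two slots, leaving two empty.
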SoundCue /Game/Sounds/Asteria/Biomes/Rainforest/Spots/Spot_Rainforest_Monkey_Calls_Night_All_Cue
root=QualityLevel_0
nodes [
 {id=Modulator_0, children [Random_0], properties {pitchMax=1.1, pitchMin=0.8, volumeMax=1.7, volumeMin=1.5}}
 {id=Modulator_2, children [Random_1], properties {pitchMax=1, pitchMin=0.8, volumeMax=1.8, volumeMin=1.5}}
 {id=Modulator_7, children [Random_8], properties {pitchMax=0.9, pitchMin=0.7, volumeMax=1.7, volumeMin=1.5}}
 {id=QualityLevel_0, children [Random_2,Random_2,null,null]}
 {
  id=Random_0
  children [WavePlayer_0, WavePlayer_1, WavePlayer_2, WavePlayer_4, WavePlayer_9, WavePlayer_10]
  properties {weights [1,1,1,1,1,1]}
 }
 {
  id=Random_1
  children [WavePlayer_12, WavePlayer_13, WavePlayer_42, WavePlayer_43, WavePlayer_44, WavePlayer_45, WavePlayer_46, WavePlayer_47]
  properties {weights [1,1,1,1,1,1,1,1]}
 }
 {id=Random_2, children [Modulator_2,Modulator_7,Modulator_0], properties {bShouldExcludeFromBranchCulling=true, weights [1,1,1]}}
 {id=Random_8, children [WavePlayer_48, WavePlayer_49, WavePlayer_50, WavePlayer_51, WavePlayer_52], properties {weights [1,1,1,1,1]}}
 {id=WavePlayer_0, properties {soundWaveAssetPtr=&/Spot_Monkey_Gibbon_Call_01}}
 {id=WavePlayer_1, properties {soundWaveAssetPtr=&/Spot_Monkey_Gibbon_Call_02}}
 {id=WavePlayer_10, properties {soundWaveAssetPtr=&/Spot_Monkey_Gibbon_Call_06}}
 {id=WavePlayer_12, properties {soundWaveAssetPtr=&/Spot_Monkey_Howler_Call_01}}
 {id=WavePlayer_13, properties {soundWaveAssetPtr=&/Spot_Monkey_Howler_Call_02}}
 {id=WavePlayer_2, properties {soundWaveAssetPtr=&/Spot_Monkey_Gibbon_Call_03}}
 {id=WavePlayer_4, properties {soundWaveAssetPtr=&/Spot_Monkey_Gibbon_Call_04}}
 {id=WavePlayer_42, properties {soundWaveAssetPtr=&/Spot_Monkey_Howler_Call_03}}
 {id=WavePlayer_43, properties {soundWaveAssetPtr=&/Spot_Monkey_Howler_Call_04}}
 {id=WavePlayer_44, properties {soundWaveAssetPtr=&/Spot_Monkey_Howler_Call_05}}
 {id=WavePlayer_45, properties {soundWaveAssetPtr=&/Spot_Monkey_Howler_Call_06}}
 {id=WavePlayer_46, properties {soundWaveAssetPtr=&/Spot_Monkey_Howler_Call_07}}
 {id=WavePlayer_47, properties {soundWaveAssetPtr=&/Spot_Monkey_Howler_Call_08}}
 {id=WavePlayer_48, properties {soundWaveAssetPtr=&/Spot_Monkey_Howler_Growl_Distant_01}}
 {id=WavePlayer_49, properties {soundWaveAssetPtr=&/Spot_Monkey_Howler_Growl_Distant_02}}
 {id=WavePlayer_50, properties {soundWaveAssetPtr=&/Spot_Monkey_Howler_Growl_Distant_03}}
 {id=WavePlayer_51, properties {soundWaveAssetPtr=&/Spot_Monkey_Howler_Growl_Distant_04}}
 {id=WavePlayer_52, properties {soundWaveAssetPtr=&/Spot_Monkey_Howler_Growl_Distant_05}}
 {id=WavePlayer_9, properties {soundWaveAssetPtr=&/Spot_Monkey_Gibbon_Call_05}}
]
properties {
 attenuationSettings=/Game/Sounds/Attenuation/Fort_TileAmbient_OneShot_Distant_Creature_Att
 bEnableBusSends=false
 maxDistance=8500
 soundClassObject=&Fort_Ambient_OneShots_Outdoors
 volumeMultiplier=0.301
 concurrencySet [/Game/Sounds/Concurrency_Rules/Asteria_Rainforest_Spots_Conc]
 preEffectBusSends [{}]
}
