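# Sound cue for a tumbleweed spot ambient one-shot in the Asteria Rainforest biome, reusing waves from the Artemis Reality Bloom WildWest POI.
# A random node picks among three equally weighted pass-by variations, and all four quality levels feed from that same node.
# It plays quietly in the outdoor ambient one-shot class, with wind tile-ambient attenuation, rainforest spot concurrency and bus sends disabled.
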SoundCue /Game/Sounds/Asteria/Biomes/Rainforest/Spots/Spot_Rainforest_Tumble_Weed_Cue
root=QualityLevel_0
nodes [
 {id=QualityLevel_0, children [Random_0,Random_0,Random_0,Random_0]}
 {id=Random_0, children [WavePlayer_3,WavePlayer_4,WavePlayer_5], properties {weights [1,1,1]}}
 {id=WavePlayer_3, properties {soundWaveAssetPtr=/Game/Sounds/Artemis/POI/Reality_Bloom_WildWest/Spots/Spot_Tumble_Weed_By_01}}
 {id=WavePlayer_4, properties {soundWaveAssetPtr=/Game/Sounds/Artemis/POI/Reality_Bloom_WildWest/Spots/Spot_Tumble_Weed_By_02}}
 {id=WavePlayer_5, properties {soundWaveAssetPtr=/Game/Sounds/Artemis/POI/Reality_Bloom_WildWest/Spots/Spot_Tumble_Weed_By_03}}
]
properties {
 attenuationSettings=/Game/Sounds/Attenuation/Fort_TileAmbient_OneShot_Wind_Att
 bEnableBusSends=false
 maxDistance=8000
 soundClassObject=&Fort_Ambient_OneShots_Outdoors
 volumeMultiplier=0.351
 concurrencySet [/Game/Sounds/Concurrency_Rules/Asteria_Rainforest_Spots_Conc]
 preEffectBusSends [{}]
}
