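# Sound cue for exterior rat squeak spot one-shots in the Asteria rainforest biome ambience.
# A quality-level node feeds the first two tiers from a random pool of fourteen squeaks and the last two from a five-squeak subset.
# Each pool passes through a modulator that varies pitch and volume.
# It uses a bird one-shot attenuation and the ambient one-shot class, with a distance-based send to the exterior ambience source bus.
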SoundCue /Game/Sounds/Asteria/Biomes/Rainforest/Spots/Spot_Rat_Squeak_Exterior_Cue
root=QualityLevel_0
nodes [
 {id=Modulator_1, children [Random_2], properties {pitchMax=1.1, volumeMax=1.1, volumeMin=0.9}}
 {id=Modulator_2, children [Random_3], properties {pitchMax=1.1, volumeMax=1.1, volumeMin=0.9}}
 {id=QualityLevel_0, children [Modulator_1,Modulator_1,Modulator_2,Modulator_2]}
 {
  id=Random_2
  children [
   WavePlayer_0
   WavePlayer_1
   WavePlayer_2
   WavePlayer_3
   WavePlayer_4
   WavePlayer_13
   WavePlayer_14
   WavePlayer_15
   WavePlayer_16
   WavePlayer_17
   WavePlayer_18
   WavePlayer_19
   WavePlayer_20
   WavePlayer_21
  ]
  properties {weights [1,1,1,1,1,1,1,1,1,1,1,1,1,1]}
 }
 {id=Random_3, children [WavePlayer_22, WavePlayer_23, WavePlayer_24, WavePlayer_36, WavePlayer_37], properties {weights [1,1,1,1,1]}}
 {id=WavePlayer_0, properties {soundWaveAssetPtr=&/Spot_Rat_Squeak_01_Exterior}}
 {id=WavePlayer_1, properties {soundWaveAssetPtr=&/Spot_Rat_Squeak_02_Exterior}}
 {id=WavePlayer_13, properties {soundWaveAssetPtr=&/Spot_Rat_Squeak_06_Exterior}}
 {id=WavePlayer_14, properties {soundWaveAssetPtr=&/Spot_Rat_Squeak_07_Exterior}}
 {id=WavePlayer_15, properties {soundWaveAssetPtr=&/Spot_Rat_Squeak_08_Exterior}}
 {id=WavePlayer_16, properties {soundWaveAssetPtr=&/Spot_Rat_Squeak_09_Exterior}}
 {id=WavePlayer_17, properties {soundWaveAssetPtr=&/Spot_Rat_Squeak_10_Exterior}}
 {id=WavePlayer_18, properties {soundWaveAssetPtr=&/Spot_Rat_Squeak_11_Exterior}}
 {id=WavePlayer_19, properties {soundWaveAssetPtr=&/Spot_Rat_Squeak_12_Exterior}}
 {id=WavePlayer_2, properties {soundWaveAssetPtr=&/Spot_Rat_Squeak_03_Exterior}}
 {id=WavePlayer_20, properties {soundWaveAssetPtr=&/Spot_Rat_Squeak_13_Exterior}}
 {id=WavePlayer_21, properties {soundWaveAssetPtr=&/Spot_Rat_Squeak_14_Exterior}}
 {id=WavePlayer_22, properties {soundWaveAssetPtr=&Spot_Rat_Squeak_09_Exterior}}
 {id=WavePlayer_23, properties {soundWaveAssetPtr=&Spot_Rat_Squeak_07_Exterior}}
 {id=WavePlayer_24, properties {soundWaveAssetPtr=&/Spot_Rat_Squeak_04_Exterior}}
 {id=WavePlayer_3, properties {soundWaveAssetPtr=&Spot_Rat_Squeak_04_Exterior}}
 {id=WavePlayer_36, properties {soundWaveAssetPtr=&Spot_Rat_Squeak_02_Exterior}}
 {id=WavePlayer_37, properties {soundWaveAssetPtr=&Spot_Rat_Squeak_14_Exterior}}
 {id=WavePlayer_4, properties {soundWaveAssetPtr=&/Spot_Rat_Squeak_05_Exterior}}
]
properties {
 attenuationSettings=&Fort_TileAmbient_OneShot_Bird_Att
 maxDistance=8005
 soundClassObject=&Fort_Ambient_OneShots
 volumeMultiplier=0.451
 busSends [{
   audioBus=null
   maxSendDistance=4500
   minSendDistance=2000
   minSendLevel=0.2
   sendLevel=1
   soundSourceBus=&AmbientAudio_Exterior_SourceBus
  }]
}
