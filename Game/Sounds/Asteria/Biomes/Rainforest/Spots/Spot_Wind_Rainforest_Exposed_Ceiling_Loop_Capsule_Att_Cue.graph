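# Looping ambient spot cue for rainforest wind at exposed ceilings, routed to the indoor ambient sound class with borrowed capsule attenuation.
# A short delay feeds a quality switch that plays only on the first two levels and leaves the other two empty.
# Two nested looping envelopers of different lengths modulate volume and pitch of one looping stereo window-wind wave, varying the loop.
# A concurrency override prevents new instances once the limit is reached.
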SoundCue /Game/Sounds/Asteria/Biomes/Rainforest/Spots/Spot_Wind_Rainforest_Exposed_Ceiling_Loop_Capsule_Att_Cue
root=Delay_0
nodes [
 {id=Delay_0, children [QualityLevel_0], properties {delayMax=0.3}}
 {
  id=Enveloper_0
  children [WavePlayer_0]
  properties {
   bLoop=true
   bLoopIndefinitely=true
   loopEnd=30
   pitchCurve {editorCurveData {keys [{interpMode=RCIM_Cubic, time=0}, <at 15.9468 0.9181 RCIM_Cubic>, {interpMode=RCIM_Cubic, time=30}]}}
   volumeCurve {editorCurveData {keys [<at 0 1>]}}
  }
 }
 {
  id=Enveloper_3
  children [Enveloper_0]
  properties {
   bLoop=true
   bLoopIndefinitely=true
   loopEnd=20
   pitchCurve {editorCurveData {keys [{time=0}]}}
   volumeCurve {editorCurveData {keys [
      {arriveTangent=-0.01, interpMode=RCIM_Cubic, time=0, value=1}
      <at 10 1.2 RCIM_Cubic>
      {interpMode=RCIM_Cubic, leaveTangent=0.01, time=20, value=1}
     ]}}
  }
 }
 {id=QualityLevel_0, children [Enveloper_3,Enveloper_3,null,null]}
 {
  id=WavePlayer_0
  properties {bLooping=true, soundWaveAssetPtr=&/Spot_Wind_Rainforest_Exposed_Window_Cold_Tonal_Stereo}
 }
]
properties {
 attenuationSettings=/Game/Sounds/Artemis/POI/Seven_Workshop/Spot_Interior_Exterior_Ceiling_Wind_Focus_Capsule_Att
 bOverrideConcurrency=true
 maxDistance=2500
 soundClassObject=&Fort_Ambient_Indoors_Athena
 volumeMultiplier=0.301
 attenuationOverrides {falloffDistance=1600}
 concurrencyOverrides {resolutionRule=PreventNew}
}
