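# Sound cue for a looping exposed-window wind spot ambience in the Asteria rainforest biome, routed to the indoor ambient sound class.
# A single looping wave player reuses the Winterfest exterior wind loop at a heavily reduced volume.
# It borrows the Hollow Base interior window wind attenuation and overrides concurrency to cap simultaneous instances at a small count.
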SoundCue /Game/Sounds/Asteria/Biomes/Rainforest/Spots/Spot_Wind_Rainforest_Exposed_Window_Loop_Capsule_02_Cue
root=WavePlayer_0
nodes [{id=WavePlayer_0, properties {bLooping=true, soundWaveAssetPtr=/Game/Athena/Sounds/Events/Winterfest/Winterfest_Exterior_Wind_Loop}}]
properties {
 attenuationSettings=/Game/Sounds/Artemis/POI/Hollow_Base/Spots/Spot_Interior_Window_Wind_Att
 bOverrideConcurrency=true
 maxDistance=1253.7231
 soundClassObject=&Fort_Ambient_Indoors_Athena
 volumeMultiplier=0.175
 concurrencyOverrides {maxCount=2}
}
